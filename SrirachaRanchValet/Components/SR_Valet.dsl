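; Vehicle radio component for Valet cars that ties streaming station playback to music events, audio analysis and quest broadcasts.
; Binds play and stop events on BeginPlay and unbinds on EndPlay.
; Regular stations add a vehicle-radio music event, register the default analysis submix and send to the Valet radio bus; player-playlist stations remove that event to avoid priority clashes.
; The owning car receives start and stop broadcasts, with start flagging station-ID changes.
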(event Custom|On%20Source%20Played
  (params (Component :type Streaming%20Radio%20Player%20Component%20Object%20Reference) (AffectedSource :type Athena%20Radio%20Station%20Structure))
  (bind (_title _stationimage _bisplayerplaylist) (Utilities|Struct|BreakAthenaRadioStation AffectedSource))
  ; The player playlist station needs to remove the radio music event so as not to clash priorities
  (if _bisplayerplaylist
    (CallFunction|RemoveMusicEvent)
    (else
      (Audio|AddEventbySynthComponent (GameInstanceSubsystems|GetMusicEventSubsystem) (Variables|Radio|GetSoundComponent) "(TagName=\"MusicEventSubsystem.Event.Radio.Vehicle\")" "(GameplayTags=)")
      (AudioAnalysis|RegisterDefaultSubmix (WorldSubsystems|GetFortAudioAnalysisSubsystem))
      (Audio|Components|Audio|SetAudioBusSendPreEffect (Variables|Radio|GetSoundComponent) "/SrirachaRanch/Audio/Valet_Radio_AudioBus.Valet_Radio_AudioBus" 1.0 :!class /Script/AudioMixer.SynthComponent)))
  ;[1 1 1] Quest ~ broadcast for radio starting
  (bind _asvalet_basic_car_radio_interface (Utilities|Casting|CastToValetBasicCarRadioInterface (Components|GetOwner (Variables|Radio|GetSoundComponent))))
  (ValetBasicCarRadioInterface|SendRadioStartEvent _asvalet_basic_car_radio_interface (CallFunction|HasStationIdChanged Component AffectedSource)))

(event Custom|On%20Source%20Stopped
  (params (Component :type Streaming%20Radio%20Player%20Component%20Object%20Reference) (AffectedSource :type Athena%20Radio%20Station%20Structure))
  (CallFunction|RemoveMusicEvent)
  ;[1 1 1] Quest ~ broadcast for radio stopping
  (bind _asvalet_basic_car_radio_interface (Utilities|Casting|CastToValetBasicCarRadioInterface (Components|GetOwner (Variables|Radio|GetSoundComponent))))
  (Class|ValetBasicCarRadioInterface|SendRadioStopEvent _asvalet_basic_car_radio_interface))

(event EventBeginPlay
  (Radio|BindEventtoOnSourcePlayed (EventDispatchers|CreateEvent :!event Custom|On%20Source%20Played))
  (Radio|BindEventtoOnSourceStopped (EventDispatchers|CreateEvent :!event Custom|On%20Source%20Stopped)))

(event EventEndPlay
  (params (EndPlayReason :type EEndPlayReason%20Enum :default "Destroyed"))
  (Radio|UnbindEventfromOnSourcePlayed (EventDispatchers|CreateEvent :!event Custom|On%20Source%20Played))
  (Radio|UnbindEventfromOnSourceStopped (EventDispatchers|CreateEvent :!event Custom|On%20Source%20Stopped))
  (bind _soundcomponent (Variables|Radio|GetSoundComponent)
    (:then
      (Audio|RemoveEventbySynthComponent (GameInstanceSubsystems|GetMusicEventSubsystem) _soundcomponent))
    (:else)))

(fn CanApplyHolster
  (params (HolsterId :type Name%20%28by%20ref%29 :default "None"))
  (return false))

(fn HasStationIdChanged
  (params (Component :type Streaming%20Radio%20Player%20Component%20Object%20Reference) (AthenaRadioStation :type Athena%20Radio%20Station%20Structure%20%28by%20ref%29))
  (bind (_title _stationimage _bisplayerplaylist _resourceid) (Utilities|Struct|BreakAthenaRadioStation AthenaRadioStation))
  (bind (_lastsource_title _lastsource_stationimage _lastsource_bisplayerplaylist _lastsource_resourceid) (Utilities|Struct|BreakAthenaRadioStation (Variables|/Script/SrirachaRanch.StreamingRadioPlayerComponent|GetLastSource_Title Component)))
  (return (Utilities|String|NotEqualExactly(String) _lastsource_resourceid _resourceid)))

(fn RemoveMusicEvent
  (Audio|RemoveEventbySynthComponent (GameInstanceSubsystems|GetMusicEventSubsystem) (Variables|Radio|GetSoundComponent))
  (AudioAnalysis|UnRegisterDefaultSubmix (WorldSubsystems|GetFortAudioAnalysisSubsystem)))
